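# EQS query that picks a reposition point for humanoid AI bots within their leash area.
# Three fallback options generate donuts around the current target, the querier, then the leash center sized by leash radius.
# Points are filtered by pathability, plus leash containment in the first two options.
# Scoring weighs line of sight to target, distances to self, enemies, nearby AI and their destinations, randomness, and strongly favors ground free of water and danger.
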EnvQuery /AICore/Humanoids/EQS/EQS_MoveAround_LeashArea
options [
 {
  generator {
   name=EnvQueryGenerator_Donut_2
   class=EnvQueryGenerator_Donut
   properties {
    bUseSpiralPattern=false
    center=&FortAthenaAIBot_EQSQueryContext_CurrentTarget
    innerRadius=300
    numberOfRings=6
    outerRadius=1000
    pointsPerRing=10
    projectionData {
     extentX=10
     navigationFilter=&NavFilter_NPC_AvoidAllWater_C
     postProjectionVerticalOffset=30
    }
   }
  }
  tests [
   FortQueryTest_InsideAIBotLeash_0 {class=FortQueryTest_InsideAIBotLeash, properties {testPurpose=Filter}}
   FortQueryTest_PathfindingBatch_1 {
    class=FortQueryTest_PathfindingBatch
    properties {
     bDefineReferenceValue=false
     boolValue=true
     clampMaxType=None
     clampMinType=None
     context=&EnvQueryContext_Querier
     filterClass=&NavFilter_NPC_AvoidAllWater_C
     filterType=Match
     floatValueMax=1000
     floatValueMin=1000
     multipleContextFilterOp=AllPass
     multipleContextScoreOp=AverageScore
     navDataOverrideContext=null
     normalizationType=Absolute
     pathFromContext=true
     referenceValue=0
     scanRangeMultiplier=1.5
     scoreClampMax=0
     scoreClampMin=0
     scoringEquation=Constant
     scoringFactor=1
     skipUnreachable=true
     testComment=Throw out things you cannot path to.
     testMode=PathExist
     testPurpose=Filter
     navFilterTag {tagName=NavFilter.NPC.DefaultMove}
    }
   }
   FortQueryTest_IsObstructed_0 {
    class=FortQueryTest_IsObstructed
    properties {
     context=&FortAthenaAIBot_EQSQueryContext_CurrentTarget
     testComment=Throw out things that don't afford line of sight to the target.
     testPurpose=Score
    }
   }
   EnvQueryTest_Distance_5 {
    class=EnvQueryTest_Distance
    properties {
     distanceTo=&EnvQueryContext_Querier
     filterType=Range
     floatValueMin=0
     scoringEquation=InverseLinear
     scoringFactor=1
     testComment=Pick point farthest from Self
     testMode=Distance3D
     testPurpose=Score
    }
   }
   EnvQueryTest_Distance_18 {
    class=EnvQueryTest_Distance
    properties {
     distanceTo=&FortAthenaAIBot_EQSQueryContext_CurrentTarget
     filterType=Range
     floatValueMin=0
     scoringEquation=Linear
     scoringFactor=1
     testComment=""
     testMode=Distance3D
     testPurpose=Score
    }
   }
   FortQueryTest_Random_1 {class=FortQueryTest_Random, properties {floatValueMax=0, scoringFactor=0.75}}
   EnvQueryTest_Distance_6 {
    class=EnvQueryTest_Distance
    properties {
     distanceTo=/Script/FortniteAIServer.FortQueryContext_AllEnemies
     filterType=Range
     floatValueMin=0
     scoringEquation=Linear
     scoringFactor=0.5
     testComment=Prefer to be farther from enemies
     testMode=Distance3D
     testPurpose=Score
    }
   }
   EnvQueryTest_Distance_7 {
    class=EnvQueryTest_Distance
    properties {
     distanceTo=&FortQueryContext_NearbyAIPawns
     filterType=Range
     floatValueMin=0
     scoringEquation=Linear
     scoringFactor=0.25
     testComment=""
     testMode=Distance3D
     testPurpose=Score
    }
   }
   EnvQueryTest_Distance_8 {
    class=EnvQueryTest_Distance
    properties {
     distanceTo=&FortQueryContext_NearbyAIPawnsMoveDestinations
     filterType=Range
     floatValueMin=0
     scoringEquation=Linear
     scoringFactor=0.25
     testComment=Try to avoid moving to other AI move to destinations
     testMode=Distance3D
     testPurpose=Score
    }
   }
   EnvQueryTest_Project_2 {
    class=EnvQueryTest_Project
    properties {
     scoringFactor=2
     testComment=Give LOTS of score to things NOT on Water. We only use Water if we really need to.
     testPurpose=Score
     projectionData {
      navigationFilter=&NavFilter_NPC_ExcludeAllWater_C
      postProjectionVerticalOffset=30
      projectDown=1024
      projectUp=1024
      traceMode=Navigation
     }
    }
   }
   EnvQueryTest_Project_3 {
    class=EnvQueryTest_Project
    properties {
     scoringFactor=0
     testComment=Project to any traversable so we don't penetrate under water.
     testPurpose=Score
     projectionData {
      navigationFilter=&NavFilter_NPC_AnyTraversable_C
      postProjectionVerticalOffset=30
      projectDown=1024
      projectUp=1024
      traceMode=Navigation
     }
    }
   }
   EnvQueryTest_Distance_9 {
    class=EnvQueryTest_Distance
    properties {
     distanceTo=&EnvQueryContext_Querier
     filterType=Range
     floatValueMin=0
     scoringEquation=Linear
     scoringFactor=0.5
     testComment=This simulates the OnCircle Generator when using a grid generator
     testMode=Distance3D
     testPurpose=Score
    }
   }
   EnvQueryTest_Project_6 {
    class=EnvQueryTest_Project
    properties {
     scoringFactor=2
     testComment=Give lots of score to things NOT on danger ground.
     testPurpose=Score
     projectionData {
      navigationFilter=/Game/Athena/AI/NavFilters/NavFilter_NPC_AnyTraversable_ExcludeDanger.NavFilter_NPC_AnyTraversable_ExcludeDanger_C
      postProjectionVerticalOffset=30
      projectDown=1024
      projectUp=1024
      traceMode=Navigation
     }
    }
   }
  ]
 }
 {
  generator {
   name=EnvQueryGenerator_Donut_0
   class=EnvQueryGenerator_Donut
   properties {
    bUseSpiralPattern=false
    center=&EnvQueryContext_Querier
    numberOfRings=6
    pointsPerRing=10
    innerRadius {
     dataBinding=&self:EnvQueryGenerator_Donut_0.AIDataProvider_QueryParams_0
     dataField=FloatValue
     defaultValue=300
    }
    outerRadius {
     dataBinding=&self:EnvQueryGenerator_Donut_0.AIDataProvider_QueryParams_1
     dataField=FloatValue
     defaultValue=1000
    }
    projectionData {
     extentX=10
     navigationFilter=&NavFilter_NPC_AvoidAllWater_C
     postProjectionVerticalOffset=30
    }
   }
  }
  tests [
   FortQueryTest_InsideAIBotLeash_1 {class=FortQueryTest_InsideAIBotLeash, properties {testPurpose=Filter}}
   FortQueryTest_PathfindingBatch_2 {
    class=FortQueryTest_PathfindingBatch
    properties {
     bDefineReferenceValue=false
     boolValue=true
     clampMaxType=None
     clampMinType=None
     context=&EnvQueryContext_Querier
     filterClass=&NavFilter_NPC_AvoidAllWater_C
     filterType=Match
     floatValueMax=1000
     floatValueMin=1000
     multipleContextFilterOp=AllPass
     multipleContextScoreOp=AverageScore
     navDataOverrideContext=null
     normalizationType=Absolute
     pathFromContext=true
     referenceValue=0
     scanRangeMultiplier=1.5
     scoreClampMax=0
     scoreClampMin=0
     scoringEquation=Constant
     scoringFactor=1
     skipUnreachable=true
     testComment=Throw out things you cannot path to.
     testMode=PathExist
     testPurpose=Filter
     navFilterTag {tagName=NavFilter.NPC.DefaultMove}
    }
   }
   FortQueryTest_IsObstructed_1 {
    class=FortQueryTest_IsObstructed
    properties {
     context=&FortAthenaAIBot_EQSQueryContext_CurrentTarget
     testComment=Throw out things that don't afford line of sight to the target.
     testPurpose=Score
    }
   }
   EnvQueryTest_Distance_13 {
    class=EnvQueryTest_Distance
    properties {
     distanceTo=&EnvQueryContext_Querier
     filterType=Range
     floatValueMin=0
     scoringEquation=Linear
     scoringFactor=1
     testComment=Pick point farthest from Self
     testMode=Distance3D
     testPurpose=Score
    }
   }
   FortQueryTest_Random_2 {class=FortQueryTest_Random, properties {floatValueMax=0, scoringFactor=0.1}}
   EnvQueryTest_Distance_14 {
    class=EnvQueryTest_Distance
    properties {
     distanceTo=&FortQueryContext_AllEnemies
     filterType=Range
     floatValueMin=0
     scoringEquation=Linear
     scoringFactor=0.5
     testComment=Prefer to be farther from enemies
     testMode=Distance3D
     testPurpose=Score
    }
   }
   EnvQueryTest_Distance_15 {
    class=EnvQueryTest_Distance
    properties {
     distanceTo=&FortQueryContext_NearbyAIPawns
     filterType=Range
     floatValueMin=0
     scoringEquation=Linear
     scoringFactor=0.25
     testComment=""
     testMode=Distance3D
     testPurpose=Score
    }
   }
   EnvQueryTest_Distance_16 {
    class=EnvQueryTest_Distance
    properties {
     distanceTo=&FortQueryContext_NearbyAIPawnsMoveDestinations
     filterType=Range
     floatValueMin=0
     scoringEquation=Linear
     scoringFactor=0.25
     testComment=Try to avoid moving to other AI move to destinations
     testMode=Distance3D
     testPurpose=Score
    }
   }
   EnvQueryTest_Project_4 {
    class=EnvQueryTest_Project
    properties {
     scoringFactor=2
     testComment=Give LOTS of score to things NOT on Water. We only use Water if we really need to.
     testPurpose=Score
     projectionData {
      navigationFilter=&NavFilter_NPC_ExcludeAllWater_C
      postProjectionVerticalOffset=30
      projectDown=1024
      projectUp=1024
      traceMode=Navigation
     }
    }
   }
   EnvQueryTest_Project_5 {
    class=EnvQueryTest_Project
    properties {
     scoringFactor=0
     testComment=Project to any traversable so we don't penetrate under water.
     testPurpose=Score
     projectionData {
      navigationFilter=&NavFilter_NPC_AnyTraversable_C
      postProjectionVerticalOffset=30
      projectDown=1024
      projectUp=1024
      traceMode=Navigation
     }
    }
   }
   EnvQueryTest_Distance_17 {
    class=EnvQueryTest_Distance
    properties {
     distanceTo=&EnvQueryContext_Querier
     filterType=Range
     floatValueMin=0
     scoringEquation=Linear
     scoringFactor=0.5
     testComment=This simulates the OnCircle Generator when using a grid generator
     testMode=Distance3D
     testPurpose=Score
    }
   }
   EnvQueryTest_Project_7 {
    class=EnvQueryTest_Project
    properties {
     scoringFactor=2
     testComment=Give lots of score to things NOT on danger ground.
     testPurpose=Score
     projectionData {
      navigationFilter=&NavFilter_NPC_AnyTraversable_ExcludeDanger_C
      postProjectionVerticalOffset=30
      projectDown=1024
      projectUp=1024
      traceMode=Navigation
     }
    }
   }
  ]
 }
 {
  generator {
   name=EnvQueryGenerator_Donut_1
   class=EnvQueryGenerator_Donut
   properties {
    bUseSpiralPattern=false
    center=/Script/FortniteAI.FortQueryContext_LeashCenter
    innerRadius=300
    numberOfRings=6
    pointsPerRing=10
    outerRadius {
     dataBinding=&self:EnvQueryGenerator_Donut_1.FortAIDataProvider_LeashComponent_2
     dataField=LeashRadius
     defaultValue=1000
    }
    projectionData {
     extentX=10
     navigationFilter=&NavFilter_NPC_AvoidAllWater_C
     postProjectionVerticalOffset=30
    }
   }
  }
  tests [
   FortQueryTest_PathfindingBatch_0 {
    class=FortQueryTest_PathfindingBatch
    properties {
     bDefineReferenceValue=false
     boolValue=true
     clampMaxType=None
     clampMinType=None
     context=&EnvQueryContext_Querier
     filterClass=&NavFilter_NPC_AvoidAllWater_C
     filterType=Match
     floatValueMax=1000
     floatValueMin=1000
     multipleContextFilterOp=AllPass
     multipleContextScoreOp=AverageScore
     navDataOverrideContext=null
     normalizationType=Absolute
     pathFromContext=true
     referenceValue=0
     scanRangeMultiplier=1.5
     scoreClampMax=0
     scoreClampMin=0
     scoringEquation=Constant
     scoringFactor=1
     skipUnreachable=true
     testComment=Throw out things you cannot path to.
     testMode=PathExist
     testPurpose=Filter
     navFilterTag {tagName=NavFilter.NPC.DefaultMove}
    }
   }
   FortQueryTest_IsObstructed_2 {
    class=FortQueryTest_IsObstructed
    properties {
     context=&FortAthenaAIBot_EQSQueryContext_CurrentTarget
     testComment=Throw out things that don't afford line of sight to the target.
     testPurpose=Score
    }
   }
   EnvQueryTest_Distance_0 {
    class=EnvQueryTest_Distance
    properties {
     distanceTo=&EnvQueryContext_Querier
     filterType=Range
     floatValueMin=0
     scoringEquation=InverseLinear
     scoringFactor=1
     testComment=Pick point closer to Self
     testMode=Distance3D
     testPurpose=Score
    }
   }
   FortQueryTest_Random_0 {class=FortQueryTest_Random, properties {floatValueMax=0, scoringFactor=0.1}}
   EnvQueryTest_Distance_1 {
    class=EnvQueryTest_Distance
    properties {
     distanceTo=&FortQueryContext_AllEnemies
     filterType=Range
     floatValueMin=0
     scoringEquation=Linear
     scoringFactor=0.5
     testComment=Prefer to be farther from enemies
     testMode=Distance3D
     testPurpose=Score
    }
   }
   EnvQueryTest_Distance_2 {
    class=EnvQueryTest_Distance
    properties {
     distanceTo=&FortQueryContext_NearbyAIPawns
     filterType=Range
     floatValueMin=0
     scoringEquation=Linear
     scoringFactor=0.25
     testComment=""
     testMode=Distance3D
     testPurpose=Score
    }
   }
   EnvQueryTest_Distance_3 {
    class=EnvQueryTest_Distance
    properties {
     distanceTo=&FortQueryContext_NearbyAIPawnsMoveDestinations
     filterType=Range
     floatValueMin=0
     scoringEquation=Linear
     scoringFactor=0.25
     testComment=Try to avoid moving to other AI move to destinations
     testMode=Distance3D
     testPurpose=Score
    }
   }
   EnvQueryTest_Project_0 {
    class=EnvQueryTest_Project
    properties {
     scoringFactor=2
     testComment=Give LOTS of score to things NOT on Water. We only use Water if we really need to.
     testPurpose=Score
     projectionData {
      navigationFilter=&NavFilter_NPC_ExcludeAllWater_C
      postProjectionVerticalOffset=30
      projectDown=1024
      projectUp=1024
      traceMode=Navigation
     }
    }
   }
   EnvQueryTest_Project_1 {
    class=EnvQueryTest_Project
    properties {
     scoringFactor=0
     testComment=Project to any traversable so we don't penetrate under water.
     testPurpose=Score
     projectionData {
      navigationFilter=&NavFilter_NPC_AnyTraversable_C
      postProjectionVerticalOffset=30
      projectDown=1024
      projectUp=1024
      traceMode=Navigation
     }
    }
   }
   EnvQueryTest_Distance_4 {
    class=EnvQueryTest_Distance
    properties {
     distanceTo=&EnvQueryContext_Querier
     filterType=Range
     floatValueMin=0
     scoringEquation=InverseLinear
     scoringFactor=0.5
     testComment=This simulates the OnCircle Generator when using a grid generator
     testMode=Distance3D
     testPurpose=Score
    }
   }
   EnvQueryTest_Project_8 {
    class=EnvQueryTest_Project
    properties {
     scoringFactor=2
     testComment=Give lots of score to things NOT on danger ground.
     testPurpose=Score
     projectionData {
      navigationFilter=&NavFilter_NPC_AnyTraversable_ExcludeDanger_C
      postProjectionVerticalOffset=30
      projectDown=1024
      projectUp=1024
      traceMode=Navigation
     }
    }
   }
  ]
 }
]
